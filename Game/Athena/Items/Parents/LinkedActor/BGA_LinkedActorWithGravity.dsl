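; Parent Blueprint for a linked building actor that stays attached to another actor and starts projectile-movement simulation when that linked actor is destroyed.
; BeginSimulation, also optionally run at BeginPlay, hands the static mesh to FortProjectileMovement and calls the OnSimulationResumed event, left empty here.
; A hit while unattached stops simulation and marks it attached again.
; The construction script configures FortLinkToActor for movement registration, rotation-based directions and forced net updates.
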(event EventBeginPlay
  (if (Variables|Default|GetSimulateOnBeginPlay)
    (CallFunction|BeginSimulation)
    (else
      (Variables|Default|SetbAttached true))))

(event Custom|BeginSimulation
  (Components|Movement|SetUpdatedComponent (Variables|Default|GetFortProjectileMovement) (Variables|Default|GetStaticMesh))
  (CallFunction|OnSimulationResumed))

(event Collision|EventHit
  (params (MyComp :type Primitive%20Component%20Object%20Reference) (Other :type Actor%20Object%20Reference) (OtherComp :type Primitive%20Component%20Object%20Reference) (bSelfMoved :type Boolean) (HitLocation :type Vector :default "0, 0, 0") (HitNormal :type Vector :default "0, 0, 0") (NormalImpulse :type Vector :default "0, 0, 0") (Hit :type Hit%20Result%20Structure%20%28by%20ref%29))
  (if (not (Variables|Default|GetbAttached))
    (Game|Components|ProjectileMovement|StopSimulating (Variables|Default|GetFortProjectileMovement) Hit)
    (Variables|Default|SetbAttached true)))

(event OnLinkedActorDestroyed(FortLinkToActor)
  (params (DamageCauser :type Actor%20Object%20Reference))
  (CallFunction|BeginSimulation)
  (Variables|Default|SetbAttached false))

(event Custom|OnSimulationResumed)

(fn ConstructionScript
  (bind _fortlinktoactor (Variables|Default|GetFortLinkToActor))
  (LinkToActorComponent|SetRegisterwithOwnerMovementComponent _fortlinktoactor true)
  (LinkToActorComponent|SetUseActorRotationforDirectionVectors _fortlinktoactor true)
  (LinkToActorComponent|SetLinkedBuildingActorstoForceNetUpdateonDie _fortlinktoactor true))
