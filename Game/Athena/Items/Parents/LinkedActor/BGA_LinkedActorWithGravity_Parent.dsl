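; Parent Blueprint for a linked gameplay actor that falls when the actor it links to is destroyed or its physics object wakes.
; Simulation, optionally started on BeginPlay, points the projectile movement component at a falling collision capsule.
; While detached, hits or water entry stop it via ShouldStop functions (true by default); stopping marks it attached again.
; Resume and stop events force net updates; death plays a replicated GameplayCue and optionally hides the actor.
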(event EventBeginPlay
  ;[0.06339 1 0.3421] Begin Play
  (|CallParentFunction)
  (if (Variables|LinkedActorWithGravityParent|GetSimulateOnBeginPlay)
    (CallFunction|BeginSimulation)
    (else
      (Variables|LinkedActorWithGravityParent|SetbAttached true))))

(event OnLinkedActorDestroyed(FortLinkToActor)
  (params (DamageCauser :type Actor%20Object%20Reference))
  ;[0.4 1 0.4314] On Linked Actor Destroyed
  ; On Linked Physics Object Wake
  (Variables|LinkedActorWithGravityParent|SetbAttached false)
  (CallFunction|BeginSimulation))

(event OnLinkedPhysicsObjectWake(FortLinkToActor)
  (params (PhysicsObject :type Actor%20Object%20Reference))
  ;[0.4 1 0.4314] On Linked Actor Destroyed
  ; On Linked Physics Object Wake
  (Variables|LinkedActorWithGravityParent|SetbAttached false)
  (CallFunction|BeginSimulation))

(event Custom|PlayDeathGC
  ;[1 1 1] Play Death GC
  (Ability|ExecuteGameplayCueReplicated (Variables|GameplayCues|GetGC_Death)))

(event Custom|BeginSimulation
  ;[0.4 1 0.4314] Begin Simulation
  (Components|Movement|SetUpdatedComponent (Variables|Default|GetFortProjectileMovement) (Variables|Default|GetFallingCollisionCapsule))
  (CallFunction|OnSimulationResumed))

(event Collision|EventHit
  (params (MyComp :type Primitive%20Component%20Object%20Reference) (Other :type Actor%20Object%20Reference) (OtherComp :type Primitive%20Component%20Object%20Reference) (bSelfMoved :type Boolean) (HitLocation :type Vector :default "0, 0, 0") (HitNormal :type Vector :default "0, 0, 0") (NormalImpulse :type Vector :default "0, 0, 0") (Hit :type Hit%20Result%20Structure%20%28by%20ref%29))
  ;[0.22 0.3657 1] Hit or Entered Water - Maybe stop simulating
  (|CallParentFunction MyComp Other OtherComp bSelfMoved HitLocation HitNormal NormalImpulse Hit)
  (if (not (Variables|LinkedActorWithGravityParent|GetbAttached))
    (bind _shouldstop (CallFunction|ShouldStopSimulationonHit Hit))
    (if _shouldstop
      (Game|Components|ProjectileMovement|StopSimulating (Variables|Default|GetFortProjectileMovement) Hit))))

(event OnEnterWaterEvent(FortWaterInteraction)
  (params (WaterBody :type Fort%20Water%20Body%20Actor%20Object%20Reference) (WaterInteractionComponent :type Fort%20Water%20Interaction%20Component%20Object%20Reference) (bIsFirstBody :type Boolean))
  ;[0.22 0.3657 1] Hit or Entered Water - Maybe stop simulating
  (if (not (Variables|LinkedActorWithGravityParent|GetbAttached))
    (bind _shouldstop (Default|ShouldStopSimulationonEnterWater WaterBody))
    (if _shouldstop
      (Game|Components|ProjectileMovement|StopSimulating (Variables|Default|GetFortProjectileMovement) (Collision|MakeHitResult false false 0.0 0.0 "0, 0, 0" "0, 0, 0" "0,0,1" "0,0,1" 0 WaterBody)))))

(event OnProjectileStop(FortProjectileMovement)
  (params (ImpactResult :type Hit%20Result%20Structure%20%28by%20ref%29))
  ;[0.22 0.3657 1] Hit or Entered Water - Maybe stop simulating
  (Variables|LinkedActorWithGravityParent|SetbAttached true)
  (CallFunction|OnSimulationStopped ImpactResult))

(event Custom|OnSimulationResumed
  ;[1 1 1] On Simulation Resumed
  (Networking|ForceNetUpdate))

(event Custom|OnSimulationStopped
  (params (StoppedHit :type Hit%20Result%20Structure))
  ;[1 1 1] On Simulation Stopped
  (Networking|ForceNetUpdate))

(event EventOnDeathServer
  (params (Damage :type Float%20%28single-precision%29) (DamageTags :type Gameplay%20Tag%20Container%20Structure%20%28by%20ref%29) (Momentum :type Vector :default "0, 0, 0") (HitInfo :type Hit%20Result%20Structure%20%28by%20ref%29) (InstigatedBy :type Controller%20Object%20Reference) (DamageCauser :type Actor%20Object%20Reference) (EffectContext :type Gameplay%20Effect%20Context%20Handle%20Structure))
  ;[1 1 1] On Death
  (|CallParentFunction Damage DamageTags Momentum HitInfo InstigatedBy DamageCauser EffectContext)
  (CallFunction|PlayDeathGC)
  (Rendering|SetActorHiddenInGame (Variables|LinkedActorWithGravityParent|GetHideOnDeath)))

(fn ShouldStopSimulationOnEnterWater
  (params (WaterBody :type Water%20Body%20Object%20Reference))
  (return true))

(fn ShouldStopSimulationOnHit
  (params (HitResult :type Hit%20Result%20Structure))
  (return true))

(fn ConstructionScript)
